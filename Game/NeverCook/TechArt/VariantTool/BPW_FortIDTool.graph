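# Editor utility widget that builds vehicle cosmetic item definitions and painted variants by product type.
# On click, it collects existing Fort item definitions or creates them with matching VCIDs from entered names.
# For each, it creates a variant token per checked paint row, then links tags, data table rows and preview textures.
# Embedded Python writes item properties and variant channels, adding a rich color variant for bodies and skins.
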EditorUtilityWidgetBlueprint /Game/NeverCook/TechArt/VariantTool/BPW_FortIDTool
graphs [
 AddCreatedVariantsToFortID {nodes [
   n1 = AddCreatedVariantsToFortID!() -> (
    then
    Fort ID:Fort Item Definition Object Reference
    Created Variants:Array of Fort Vehicle Cosmetics Variant Token Type Object References
    VCIDAsset:Vehicle Cosmetics Item Definition Object Reference
   ) {is=title K2Node_FunctionEntry}
   > Set LocalVCIDAsset(LocalVCIDAsset:Vehicle Cosmetics Item Definition Object Reference=@n1.VCIDAsset) -> (Output_Get:Vehicle Cosmetics Item Definition Object Reference)
   > SetInitialVehicleCosmeticProperties!(
    execute
    self
    FortID:Fort Item Definition Object Reference=@n1
    VehicleCosmetic:Vehicle Cosmetics Item Definition Object Reference=@n1.VCIDAsset
   ) -> (then)
   > n2 = For Each Loop(Array:Array of Fort Vehicle Cosmetics Variant Token Type Object References=@n1.Created Variants) -> (Array Element:Fort Vehicle Cosmetics Variant Token Type Object Reference)
   > SetVariantProperties!(
    execute
    self
    PaintName:Name="Get (a ref)"(Array:Array of Names=Get PaintNames() -> (PaintNames:Array of Names), Dimension 1=@n2.Array Index) -> (Output:"Name (by ref)")
    FortID:Fort Item Definition Object Reference=@n1
    Variant:Fort Vehicle Cosmetics Variant Token Type Object Reference=@n2
    VCIDAsset:Vehicle Cosmetics Item Definition Object Reference=Get LocalVCIDAsset() -> (LocalVCIDAsset:Vehicle Cosmetics Item Definition Object Reference)
   ) -> (then)
  ]}
 Colllect Fort IDs from Selection {nodes [
   Colllect Fort IDs from Selection() {is=title K2Node_FunctionEntry}
   > Branch(Condition=IsValid(Object=Get OverwriteExistingID() -> (OverwriteExistingID:Fort Item Definition Object Reference)))
   > n1 = For Each Loop(Array:Array of Object References) -> (Array Element:Object Reference)
   Return Node!(
    execute=@n1.Completed
    Succeeded:bool=Is Not Empty!(TargetArray:Array of Fort Item Definition Object References=Get AllFortIDs() -> (AllFortIDs:Array of Fort Item Definition Object References)) -> (ReturnValue:bool)
   )
   > Branch(Condition=IsA!(Object:Object Reference=@n1, ObjectClass:Object Class Reference=/Script/FortniteGame.FortItemDefinition) -> (ReturnValue:bool))
   > n2 = Cast To FortItemDefinition!(execute, Object=@n1) -> (then, CastFailed, AsFort Item Definition:Fort Item Definition Object Reference) {is=title K2Node_DynamicCast}
   > Add(TargetArray:Array of Fort Item Definition Object References=Get AllFortIDs(), NewItem:"Fort Item Definition Object Reference (by ref)"=@n2) {is=title K2Node_CallArrayFunction}
   Reroute Node(InputPin:Wildcard) -> (OutputPin:Wildcard)
  ]}
 Creat Vehicle Cosmetic ID {nodes [
   n1 = Creat Vehicle Cosmetic ID!() -> (then, AssetName:String) {is=title K2Node_FunctionEntry}
   > Set LocalAssetName(LocalAssetName:String=@n1) -> (Output_Get:String)
   > n2 = Branch(Condition=IsEmpty(InString=@n1))
   n3 = GetVehicleCosmeticIDClass!(execute=@n2.else, self) -> (then, NewParam:Vehicle Cosmetics Item Definition Soft Class Reference)
   n4 = Get ProductTypeSelection() -> (ProductTypeSelection:DelMar_ProductTypeEnum Enum=NewEnumerator0)
   n5 = "Equal (Enum)"(A:DelMar_ProductTypeEnum Enum=@n4, B:DelMar_ProductTypeEnum Enum=NewEnumerator3)
   n6 = GetAssetTypePath!(
    Product Type:DelMar_ProductTypeEnum Enum=@n4
    Product Name:String=@n1
    Parent Name:String=Get Parent Name() -> (Parent Name:String)
    Include Product Subfolder:bool=NOT Boolean(A=@n5)
    Include Parent Subfolder:bool=@n5
   ) -> (Path:String)
   n7 = Enum to String!(Enumerator:DelMar_ProductTypeEnum Enum=@n4) -> (ReturnValue:String)
   n8 = "Equal (Enum)"(A=@n4, B=NewEnumerator4)
   n9 = Get AntennaIsFlagSubclass() -> (AntennaIsFlagSubclass:bool)
   > n10 = Get or Create Data Asset!(
    execute
    Path:String=Select(
     Option 0:String=Append!(A=@n6, B=/VCID_, C:String=@n7, D:String=_, E:String=@n1) -> (ReturnValue)
     Option 1:String=Append!(A=@n6, B=/VCID_, C=@n7, D=_, E=Flag, F:String=@n1) -> (ReturnValue)
     Index=AND Boolean(A=@n8, B=@n9)
    ) -> (ReturnValue:String)
    Class:Data Asset Class Reference=Resolve Soft Reference!(Input:Vehicle Cosmetics Item Definition Soft Class Reference=@n3) -> (Output:Vehicle Cosmetics Item Definition Class Reference)
   ) -> (then, bSuccess:bool, ReturnValue:Object Reference) {is=title K2Node_CallFunction}
   > n11 = Cast To VehicleCosmeticsItemDefinition!(execute, Object=@n10.ReturnValue) -> (then, CastFailed, AsVehicle Cosmetics Item Definition:Vehicle Cosmetics Item Definition Object Reference) {is=title K2Node_DynamicCast}
   > n12 = Set CreatedVCID(CreatedVCID:Vehicle Cosmetics Item Definition Object Reference=@n11) -> (Output_Get:Vehicle Cosmetics Item Definition Object Reference)
   > Add(
    TargetArray:Array of Vehicle Cosmetics Item Definition Object References=Get AllVCIDs() -> (AllVCIDs:Array of Vehicle Cosmetics Item Definition Object References)
    NewItem:"Vehicle Cosmetics Item Definition Object Reference (by ref)"=@n12
   )
   n13 = Get CreatedVCID() -> (CreatedVCID:Vehicle Cosmetics Item Definition Object Reference)
   > n14 = Branch(Condition=IsA(Object=@n13, ObjectClass=/Script/VehicleCosmeticsRuntime.VehicleCosmeticsItemDefinition_Body))
   n15 = Get LocalAssetName() -> (LocalAssetName:String)
   n16 = Get CreatedVCID()
   > SetupBodyVCID!(execute, self, VCID:Vehicle Cosmetics Item Definition Object Reference=@n16, AssetName:String=@n15) -> (then)
   n17 = Get CreatedVCID()
   > Return Node!(execute, VCID:Vehicle Cosmetics Item Definition Object Reference=@n17)
   n18 = Branch(execute=@n14.else, Condition=IsA(Object=@n13, ObjectClass=/Script/VehicleCosmeticsRuntime.VehicleCosmeticsItemDefinition_Skin))
   > SetupSkinVCID!(execute, self, VCID:Vehicle Cosmetics Item Definition Object Reference=@n16, AssetName:String=@n15) -> (then)
   > Return Node(VCID=@n17)
   n19 = Branch(execute=@n18.else, Condition=IsA(Object=@n13, ObjectClass=/Script/VehicleCosmeticsRuntime.VehicleCosmeticsItemDefinition_Wheel))
   > SetupWheelVCID!(execute, self, VCID:Vehicle Cosmetics Item Definition Object Reference=@n16, AssetName:String=@n15) -> (then)
   > Return Node(VCID=@n17)
   n20 = Branch(execute=@n19.else, Condition=IsA(Object=@n13, ObjectClass=/Script/VehicleCosmeticsRuntime.VehicleCosmeticsItemDefinition_Booster))
   > SetupBoosterVCID!(execute, self, VCID:Vehicle Cosmetics Item Definition Object Reference=@n16, AssetName:String=@n15) -> (then)
   n21 = Get CreatedVCID()
   > Return Node(VCID=@n21)
   Branch(execute=@n20.else, Condition=IsA(Object=@n13, ObjectClass=/Script/VehicleCosmeticsRuntime.VehicleCosmeticsItemDefinition_DriftTrail))
   > SetupDriftTrail!(execute, self, VCID:Vehicle Cosmetics Item Definition Object Reference=@n16, AssetName:String=@n15) -> (then)
   > Return Node(VCID=@n21)
  ]}
 Create New Fort IDAsset {nodes [
   n1 = Create New Fort IDAsset!() -> (then, InFortIDName:String) {is=title K2Node_FunctionEntry}
   > Set FortIDName(FortIDName:String=@n1) -> (Output_Get:String)
   > n2 = Branch(Condition=IsEmpty(InString=@n1))
   > Return Node!(execute, FortID:Fort Item Definition Object Reference)
   n3 = GetAssetTools(execute=@n2.else)
   > n4 = GetVehicleIDClass!(execute, self) -> (then, NewParam:Fort Vehicle Cosmetics Item Definition Class Reference)
   n5 = Get ProductTypeSelection() -> (ProductTypeSelection=NewEnumerator0)
   > n6 = GetModifiedFortIDName!(execute, self, InProductTypeName:String=Enum to String(Enumerator=@n5)) -> (then, ProductTypeName:String)
   n7 = Get FortIDName() -> (FortIDName:String)
   n8 = "Equal (Enum)"(A=@n5, B=NewEnumerator4)
   n9 = Get AntennaIsFlagSubclass()
   n10 = Get FortIDName()
   n11 = "Equal (Enum)"(A=@n5, B=NewEnumerator3)
   > n12 = CreateAsset(
    self=@n3
    AssetName=Select(
     Option 0=Append!(A=@n6, B=_, C=@n10) -> (ReturnValue)
     Option 1=Append!(A=@n6, B=_, C=Flag, D:String=@n7) -> (ReturnValue)
     Index=AND Boolean(A=@n8, B=@n9)
    )
    PackagePath=GetAssetTypePath(
     Product Type=@n5
     Product Name=Get FortIDName()
     Parent Name=Get Parent Name()
     Include Product Subfolder=NOT Boolean(A=@n11)
     Include Parent Subfolder=@n11
    )
    AssetClass=@n4
   )
   > n13 = Cast To FortItemDefinition(Object=@n12)
   > Add(TargetArray:Array of Fort Item Definition Object References=Get AllFortIDs(), NewItem:"Fort Item Definition Object Reference (by ref)"=@n13)
   > Return Node(FortID=@n13)
   PrintString(execute=@n13.CastFailed, InString="WARNING: Vehicle ID failed to be created or found!", TextColor=(R=1,G=0,B=0.0646,A=1), Duration=5)
  ]}
 CreateFortIDsFromInput {nodes [
   CreateFortIDsFromInput() {is=title K2Node_FunctionEntry}
   > Branch(Condition=Is Not Empty(TargetArray:Array of Strings=Get CreateNewFortIDs() -> (CreateNewFortIDs:Array of Strings)))
   > n1 = For Each Loop(Array:Array of Strings=Get CreateNewFortIDs()) -> (Array Element:String)
   Return Node!(execute=@n1.Completed, bSucceeded:bool=Is Not Empty(TargetArray:Array of Fort Item Definition Object References=Get AllFortIDs()))
   > n2 = Creat Vehicle Cosmetic ID!(execute, self, AssetName:String=@n1) -> (then, VCID:Vehicle Cosmetics Item Definition Object Reference) {is=title K2Node_CallFunction}
   > n3 = Create New Fort IDAsset!(execute, self, InFortIDName:String=@n1) -> (then, FortID:Fort Item Definition Object Reference) {is=title K2Node_CallFunction}
   > SetInitialFortIDProperties!(execute, self, FortID:Fort Item Definition Object Reference=@n3, VCID:Vehicle Cosmetics Item Definition Object Reference=@n2) -> (then)
  ]}
 CreateRichColorVariantProperty {nodes [
   n1 = CreateRichColorVariantProperty!() -> (then, FortID:Fort Item Definition Object Reference) {is=title K2Node_FunctionEntry}
   n2 = Enum to String(Enumerator=Get ProductTypeSelection() -> (ProductTypeSelection=NewEnumerator0))
   > Branch(Condition=OR Boolean(A=Contains(SearchIn=@n2, Substring=Body), B=Contains(SearchIn=@n2, Substring=Skin)))
   > n3 = GetVariantChannelName!(execute, self, InFortID:Fort Item Definition Object Reference=@n1) -> (then, VariantChannelName:String)
   > n4 = LoadAsset(AssetPath=/VehicleCosmetics/Mutable/Bodies/CS_Bodies_PrimaryColor)
   n5 = Enum to String(Enumerator=Get ProductTypeSelection() -> (ProductTypeSelection=NewEnumerator0))
   > Execute Python Script!(
    execute
    PythonScript="import unreal\r\n\r\nvariants = fort_id_asset.get_editor_property(\"ItemVariants\")\r\nnew_variant = unreal.FortCosmeticRichColorVariant(fort_id_asset)\r\n\r\nloc_channel_name = unreal.NSLOCTEXT('', fort_id_asset.get_name() + \".\" + new_variant.get_name(), variant_channel_name)\r\n\r\nnew_variant.set_editor_property(\r\n                \"VariantChannelName\",\r\n                loc_channel_name)\r\n\r\ntagChannel = unreal.FortniteEditorLibrary.request_gameplay_tag(variant_channel_tag, True)\r\n\r\nnew_variant.set_editor_property(\r\n                \"VariantChannelTag\",\r\n                tagChannel)\r\n\r\ninlineVariantDef = new_variant.get_editor_property(\"InlineVariant\")\r\nrichColorVar = inlineVariantDef.get_editor_property(\"RichColorVar\")\r\n\r\nrichColorVar.set_editor_property(\r\n                \"ColorSwatchForChoices\",\r\n                color_swatch_choice)\r\n\r\nrichColorVar.set_editor_property(\r\n                \"ColorParamName\",\r\n                asset_type_name + \"Color\")\r\n\r\nrichColorVar.set_editor_property(\r\n                \"DefaultStartingColor\", \r\n                 unreal.LinearColor(0.82, 0.82, 0.82, 1))\r\n\r\nvariants.append(new_variant)"
    FortIDAsset:Fort Item Definition Object Reference=@n1
    VariantChannelName:String=@n3
    VariantChannelTag:String=Append!(A=Cosmetics.Variant.Channel.Vehicle., B=@n5, C=.Color) -> (ReturnValue)
    ColorSwatchChoice:Object Reference=@n4
    AssetTypeName:String=@n5
   ) -> (then, ReturnValue) {inputs [FortIDAsset, VariantChannelName, VariantChannelTag, ColorSwatchChoice, AssetTypeName]}
  ]}
 EventGraph {nodes [
   "On Clicked (Button_CreatePaintedVariants)"() {is=title K2Node_ComponentBoundEvent}
   > HandleVariableInitialization()
   > GetSelectedAssetAndInitVars()
   > NameAndGenerateVariants()
   GetSelectedAssetAndInitVars() {is=title K2Node_CustomEvent}
   > n1 = Colllect Fort IDs from Selection!(execute, self) -> (then, Succeeded:bool) {is=title K2Node_CallFunction}
   > n2 = Branch(Condition=@n1)
   > n3 = PrintString(InString=SUCCESS: Correct ID/ID's Selected., TextColor=(R=0,G=0.66,B=0.126,A=1), Duration=5)
   SetDebugStrings!(
    execute=[@n3.then, @n9.then]
    self
    Fort ID:Fort Item Definition Object Reference="Get (a copy)"(Array:Array of Fort Item Definition Object References=Get AllFortIDs()) -> (Output:Fort Item Definition Object Reference)
   ) -> (then)
   > n4 = Get Checked Named!(execute, self:Object Reference=Get DelMar_ListDataTableCheckBoxes() -> (DelMar_ListDataTableCheckBoxes:Object Reference)) -> (then, CheckedNames:Array of Names) {is=title K2Node_CallFunction}
   > n5 = Set PaintNames(PaintNames:Array of Names=@n4) -> (Output_Get:Array of Names)
   > n6 = Branch(Condition=Is Not Empty(TargetArray:Array of Names=@n5))
   > n7 = Insert!(execute, TargetArray:Array of Names=Get PaintNames(), NewItem:"Name (by ref)", Index:int=0) -> (then) {is=title K2Node_CallArrayFunction}
   Set CurrentSelectionPath(execute=[@n7.then, @n6.else], CurrentSelectionPath:String=GetCurrentContentBrowserPath!() -> (OutPath:String, ReturnValue:bool)) -> (Output_Get:String)
   n8 = CreateFortIDsFromInput!(execute=@n2.else, self) -> (then, bSucceeded:bool) {is=title K2Node_CallFunction}
   > n9 = Branch(Condition=@n8)
   OnProductTypeChange!() -> (OutputDelegate, then, PropertyName:Name) {is=title K2Node_CustomEvent}
   n10 = Get ProductTypeSelection() -> (ProductTypeSelection=NewEnumerator0)
   > SetIsEnabled(
    self=Get ParentNameDetailView() -> (ParentNameDetailView:Details View Object Reference)
    bInIsEnabled=OR Boolean(A="Equal (Enum)"(A=@n10, B=NewEnumerator3), B="Equal (Enum)"(A=@n10, B=NewEnumerator4))
   )
   NameAndGenerateVariants() {is=title K2Node_CustomEvent}
   > n11 = For Each Loop(Array:Array of Fort Item Definition Object References=Get AllFortIDs()) -> (Array Element:Fort Item Definition Object Reference)
   > GenerateVariantsForFortID!(execute, self, CurrentFortID:Fort Item Definition Object Reference=@n11) -> (then)
   > AddCreatedVariantsToFortID!(
    execute
    self
    Fort ID:Fort Item Definition Object Reference=@n11
    Created Variants:Array of Fort Vehicle Cosmetics Variant Token Type Object References=Get CreatedFortVariants() -> (CreatedFortVariants:Array of Fort Vehicle Cosmetics Variant Token Type Object References)
    VCIDAsset:Vehicle Cosmetics Item Definition Object Reference="Get (a copy)"(Array:Array of Vehicle Cosmetics Item Definition Object References=Get AllVCIDs(), Dimension 1=@n11.Array Index) -> (Output:Vehicle Cosmetics Item Definition Object Reference)
   ) -> (then) {is=title K2Node_CallFunction}
   > CreateRichColorVariantProperty!(execute, self, FortID:Fort Item Definition Object Reference=@n11) -> (then) {is=title K2Node_CallFunction}
   Event Construct()
   > n12 = Set VehicleCosmeticDataTable(VehicleCosmeticDataTable:Data Table Object Reference=/VehicleCosmetics/DataTables/DT_VehicleCosmetics_Painted) -> (Output_Get:Data Table Object Reference)
   > Set Data Table!(execute, self:Object Reference=Get DelMar_ListDataTableCheckBoxes(), InTable:Data Table Object Reference=@n12) -> (then) {is=title K2Node_CallFunction}
   > OnProductTypeChange!(execute, self, PropertyName:Name=None) -> (then) {is=title K2Node_CallFunction}
   Event PreConstruct()
   > SetObject(self=Get RarityDetailView() -> (RarityDetailView:Details View Object Reference), NewObject=Self-Reference())
   n13 = Get ProductTypeDetailView() -> (ProductTypeDetailView:Details View Object Reference)
   > SetObject(self=@n13, NewObject=Self-Reference())
   > Bind Event to On Property Changed(self:Property View Base Object Reference=@n13, Delegate=Create Event()) {is=title K2Node_AddDelegate}
   > SetObject(self=Get FortIDInputDetailView() -> (FortIDInputDetailView:Details View Object Reference), NewObject=Self-Reference())
   > SetObject(self=Get ParentNameDetailView(), NewObject=Self-Reference())
   n14 = Construct Object from Class!(
    execute
    Class=&/EUT_GenerateVariantsForFortId.EUT_GenerateVariantsForFortId_C
    self
    CurrentSelectionPath:String=Get CurrentSelectionPath() -> (CurrentSelectionPath:String)
    PaintNames:Array of Names=Get PaintNames()
    ProductTypeSelection:DelMar_ProductTypeEnum Enum=Get ProductTypeSelection() -> (ProductTypeSelection=NewEnumerator0)
    VehicleCosmeticDataTable:Data Table Object Reference=Get VehicleCosmeticDataTable() -> (VehicleCosmeticDataTable:Data Table Object Reference)
    AllFortIDs:Array of Fort Item Definition Object References=Get AllFortIDs()
    AllVCIDs:Array of Vehicle Cosmetics Item Definition Object References=Get AllVCIDs()
    RaritySelection:EFortRarity Enum=Get RaritySelection() -> (RaritySelection:EFortRarity Enum=Common)
   ) -> (then, ReturnValue:EUT Generate Variants For Fort Id Object Reference)
   > RegisterAndExecuteTask!(
    execute
    self:Editor Utility Subsystem Object Reference=Get EditorUtilitySubsystem()
    NewTask:Editor Utility Task Object Reference=@n14
    OptionalParentTask:Editor Utility Task Object Reference
   ) -> (then)
  ]}
 GenerateVariantsForFortID {nodes [
   n1 = GenerateVariantsForFortID!() -> (then, CurrentFortID:Fort Item Definition Object Reference) {is=title K2Node_FunctionEntry}
   > n2 = Set FortID(FortID:Fort Item Definition Object Reference=@n1) -> (Output_Get:Fort Item Definition Object Reference)
   > Clear(TargetArray:Array of Fort Vehicle Cosmetics Variant Token Type Object References=Get CreatedFortVariants())
   > n3 = GetRelevantStringsFromID!(execute, self, FortID:Fort Item Definition Object Reference=@n2) -> (then, CurrentIDName:String, CurrentVehicleName:String, CurrentPathAndName:String)
   > Set IDName(IDName:String=@n3) -> (Output_Get:String)
   > Set VehicleName(VehicleName:String=@n3.CurrentVehicleName) -> (Output_Get:String)
   > Set PathAndName(PathAndName:String=@n3.CurrentPathAndName) -> (Output_Get:String)
   > n4 = GetAssetTools()
   > n5 = For Each Loop(Array:Array of Names=Get PaintNames()) -> (Array Element:Name)
   > n6 = GetAssetPrefixName!(execute, self, AssetName:String=Enum to String(Enumerator=Get ProductTypeSelection() -> (ProductTypeSelection=NewEnumerator0))) -> (then, Prefix:String)
   n7 = Replace(SourceString="To String (Name)"(InName=@n5), From=" ")
   n8 = Get ProductTypeSelection() -> (ProductTypeSelection=NewEnumerator0)
   n9 = "Equal (Enum)"(A=@n8, B=NewEnumerator3)
   n10 = Get VehicleName() -> (VehicleName:String)
   n11 = Get Parent Name()
   > n12 = CreateAsset(
    self=@n4
    AssetName=Append!(A=@n6, B=_, C=Get VehicleName(), D:String=_, E:String=@n7) -> (ReturnValue)
    PackagePath=Append!(
     A=GetAssetTypePath(Product Type=@n8, Product Name=@n10, Parent Name=@n11, Include Product Subfolder=true, Include Parent Subfolder=@n9)
     B=/
     C=Variants
    ) -> (ReturnValue)
    AssetClass=Get Class(Object=/Game/NeverCook/TechArt/VariantTool/_TemplateFiles/PaintedVariant_Template) -> (ReturnValue:Object Class Reference)
   )
   > n13 = Cast To FortVehicleCosmeticsVariantTokenType!(execute, Object=@n12) -> (then, CastFailed, AsFort Vehicle Cosmetics Variant Token Type:Fort Vehicle Cosmetics Variant Token Type Object Reference) {is=title K2Node_DynamicCast}
   > Add(
    TargetArray:Array of Fort Vehicle Cosmetics Variant Token Type Object References=Get CreatedFortVariants()
    NewItem:"Fort Vehicle Cosmetics Variant Token Type Object Reference (by ref)"=@n13
   )
   > Execute Python Script!(
    execute
    PythonScript="import unreal\r\n\r\ntagName = unreal.FortniteEditorLibrary.request_gameplay_tag(\"Cosmetics.Variant.Property.Vehicle.Painted.\" + my_color, True)\r\ntagChannel = unreal.FortniteEditorLibrary.request_gameplay_tag(\"Cosmetics.Variant.Channel.Vehicle.Painted\", True)\r\n\r\nmy_asset.set_editor_property(\r\n                \"grant_base_item\",\r\n                True)\r\n\r\nmy_asset.set_editor_property(\"cosmetic_item\", fort_id)\r\nmy_asset.set_editor_property(\"VariantNameTag\", tagName)\r\nmy_asset.set_editor_property(\"VariantChannelTag\", tagChannel)\r\nmy_asset.set_editor_property(\"Rarity\", rarity)"
    my_asset:Fort Vehicle Cosmetics Variant Token Type Object Reference=@n13
    my_color:String=@n7
    SelAssetInput:String=Get PathAndName() -> (PathAndName:String)
    FortID:Fort Item Definition Object Reference=Get FortID() -> (FortID:Fort Item Definition Object Reference)
    Rarity:EFortRarity Enum=Get RaritySelection() -> (RaritySelection=Common)
   ) -> (then, ReturnValue) {inputs [my_asset, my_color, SelAssetInput, FortID, Rarity]}
  ]}
 GetAssetPrefixName {nodes [
   n1 = GetAssetPrefixName!() -> (then, AssetName:String) {is=title K2Node_FunctionEntry}
   > Set LocalAssetName(LocalAssetName=@n1)
   > n2 = Branch(Condition=Contains(SearchIn=@n1, Substring=Wheel))
   n3 = Get LocalAssetName()
   > Return Node!(execute, Prefix:String=@n3)
   Return Node(execute=@n2.else, Prefix=Append!(A=Car, B=@n3) -> (ReturnValue))
  ]}
 GetAssetTypePath {nodes [
   n1 = GetAssetTypePath!() -> (
    then
    Product Type:DelMar_ProductTypeEnum Enum=NewEnumerator0
    Product Name:String
    Parent Name:String
    Include Product Subfolder:bool=~
    Include Parent Subfolder:bool=~
   ) {is=title K2Node_FunctionEntry}
   n2 = Select!(
    NewEnumerator0:String=/VehicleCosmetics/Mutable/Bodies
    NewEnumerator1:String=/VehicleCosmetics/Mutable/DriftTrail/ItemDefs
    NewEnumerator2:String=/VehicleCosmetics/Mutable/Boosters/ItemDefs
    NewEnumerator3:String=/VehicleCosmetics/Mutable/Skins
    NewEnumerator4:String=Select!(Option 0:String=/DealershipAntennas, Option 1:String=/DealershipAntennas/_Flags, Index:bool=Get AntennaIsFlagSubclass()) -> (ReturnValue)
    NewEnumerator5:String
    NewEnumerator6:String=/VehicleCosmetics/Mutable/Wheels
    Index:DelMar_ProductTypeEnum Enum=@n1
   ) -> (ReturnValue)
   n3 = Select(Option 0=@n2, Option 1=Append!(A=@n2, B=/, C:String=@n1.Parent Name) -> (ReturnValue), Index=@n1.Include Parent Subfolder)
   > Return Node!(execute, Path:String=Select(Option 0=@n3, Option 1=Append(A=@n3, B=/, C=@n1.Product Name), Index=@n1.Include Product Subfolder))
  ]}
 GetGameplayTagFromPaintName {nodes [
   n1 = GetGameplayTagFromPaintName!() -> (then, PaintName:Name) {is=title K2Node_FunctionEntry}
   n2 = "To String (Name)"(InName=@n1)
   n3 = MakeLiteralString(Value=Cosmetics.Variant.Property.Vehicle.Painted.)
   > n4 = RequestGameplayTag!(execute, TagName:String=Replace(SourceString=Append!(A=@n3, B=@n2) -> (ReturnValue), From=" "), bErrorIfNotFound:bool=true) -> (then, OutTag:Gameplay Tag Structure, ReturnValue:bool)
   > Return Node!(execute, GameplayTag:Gameplay Tag Structure=@n4)
  ]}
 GetGameplayTagFromProductType {nodes [
   GetGameplayTagFromProductType() {is=title K2Node_FunctionEntry}
   n1 = Get ProductTypeSelection() -> (ProductTypeSelection=NewEnumerator0)
   > n2 = RequestGameplayTag(TagName=Append(A=MakeLiteralString(Value=VehicleCosmetics_), B=Enum to String(Enumerator=@n1)), bErrorIfNotFound=false)
   > Return Node(GameplayTag=@n2)
  ]}
 GetModifiedFortIDName {nodes [
   n1 = GetModifiedFortIDName!() -> (then, InProductTypeName:String) {is=title K2Node_FunctionEntry}
   > Set FortIDName(FortIDName=@n1)
   > n2 = Switch on DelMar_ProductTypeEnum!(execute, Selection:DelMar_ProductTypeEnum Enum=Get ProductTypeSelection() -> (ProductTypeSelection=NewEnumerator0)) -> (
    NewEnumerator0:Exec
    NewEnumerator1:Exec
    NewEnumerator2:Exec
    NewEnumerator3:Exec
    NewEnumerator4:Exec
    NewEnumerator5:Exec
    NewEnumerator6:Exec
   ) {is=title K2Node_SwitchEnum}
   Return Node!(execute=[@n2.NewEnumerator0, @n2.NewEnumerator3], ProductTypeName:String=Append(A=Car, B=Get FortIDName()))
   Return Node(
    execute=[@n2.NewEnumerator1, @n2.NewEnumerator2, @n2.NewEnumerator4, @n2.NewEnumerator5, @n2.NewEnumerator6]
    ProductTypeName=Get FortIDName()
   )
   n3 = Get AntennaIsFlagSubclass()
   Return Node(ProductTypeName=Append(A=Select(Option 0=Antenna, Option 1=Antenna_Flag, Index=@n3), B=Get FortIDName()))
  ]}
 GetPaintedRowTag {nodes [
   GetPaintedRowTag() {is=title K2Node_FunctionEntry}
   > n1 = RequestGameplayTag(TagName=MakeLiteralString(Value=CosmeticProperty.Vehicle.PaintedRow), bErrorIfNotFound=true)
   > Return Node!(execute, GameplayTag:Gameplay Tag Structure=@n1)
  ]}
 GetPreviewTextureForColor {nodes [
   n1 = GetPreviewTextureForColor!() -> (then, PaintName:Name) {is=title K2Node_FunctionEntry}
   > Set LocalPaintString(LocalPaintString:String="To String (Name)"(InName=@n1)) -> (Output_Get:String)
   n2 = Get LocalPaintString() -> (LocalPaintString:String)
   > n3 = Branch(Condition=Contains(SearchIn=@n2, Substring=Burnt Sienna))
   > Return Node!(execute, FoundTexture:Texture Object Reference=/VehicleCosmetics/UI/Icons/Painted/T_UI_Icon_Painted_BurntSiena)
   n4 = Branch(execute=@n3.else, Condition=Contains(SearchIn=@n2, Substring=Black))
   > Return Node(FoundTexture=/VehicleCosmetics/UI/Icons/Painted/T_UI_Icon_Painted_Black)
   n5 = Branch(execute=@n4.else, Condition=Contains(SearchIn=@n2, Substring=Cobalt))
   > Return Node(FoundTexture=/VehicleCosmetics/UI/Icons/Painted/T_UI_Icon_Painted_Cobalt)
   n6 = Branch(execute=@n5.else, Condition=Contains(SearchIn=@n2, Substring=Crimson))
   > Return Node(FoundTexture=/VehicleCosmetics/UI/Icons/Painted/T_UI_Icon_Painted_Crimson)
   n7 = Branch(execute=@n6.else, Condition=Contains(SearchIn=@n2, Substring=Forest Green))
   > Return Node(FoundTexture=/VehicleCosmetics/UI/Icons/Painted/T_UI_Icon_Painted_ForestGreen)
   n8 = Get LocalPaintString()
   n9 = Branch(execute=@n7.else, Condition=Contains(SearchIn=@n8, Substring=Gold))
   > Return Node(FoundTexture=/VehicleCosmetics/UI/Icons/Painted/T_UI_Icon_Painted_Gold)
   n10 = Branch(execute=@n9.else, Condition=Contains(SearchIn=@n8, Substring=Grey))
   > Return Node(FoundTexture=/VehicleCosmetics/UI/Icons/Painted/T_UI_Icon_Painted_Grey)
   n11 = Branch(execute=@n10.else, Condition=Contains(SearchIn=@n8, Substring=Lime))
   > Return Node(FoundTexture=/VehicleCosmetics/UI/Icons/Painted/T_UI_Icon_Painted_Lime)
   n12 = Branch(execute=@n11.else, Condition=Contains(SearchIn=@n8, Substring=Orange))
   > Return Node(FoundTexture=/VehicleCosmetics/UI/Icons/Painted/T_UI_Icon_Painted_Orange)
   n13 = Branch(execute=@n12.else, Condition=Contains(SearchIn=@n8, Substring=Pink))
   > Return Node(FoundTexture=/VehicleCosmetics/UI/Icons/Painted/T_UI_Icon_Painted_Pink)
   n14 = Get LocalPaintString()
   n15 = Branch(execute=@n13.else, Condition=Contains(SearchIn=@n14, Substring=Purple))
   > Return Node(FoundTexture=/VehicleCosmetics/UI/Icons/Painted/T_UI_Icon_Painted_Purple)
   n16 = Branch(execute=@n15.else, Condition=Contains(SearchIn=@n14, Substring=Saffron))
   > Return Node(FoundTexture=/VehicleCosmetics/UI/Icons/Painted/T_UI_Icon_Painted_Saffron)
   n17 = Branch(execute=@n16.else, Condition=Contains(SearchIn=@n14, Substring=Sky Blue))
   > Return Node(FoundTexture=/VehicleCosmetics/UI/Icons/Painted/T_UI_Icon_Painted_SkyBlue)
   Branch(execute=@n17.else, Condition=Contains(SearchIn=@n14, Substring=Titanium White))
   > Return Node(FoundTexture=/VehicleCosmetics/UI/Icons/Painted/T_UI_Icon_Painted_TitaniumWhite)
  ]}
 GetRelevantStringsFromID {nodes [
   n1 = GetRelevantStringsFromID!() -> (then, FortID:Fort Item Definition Object Reference) {is=title K2Node_FunctionEntry}
   > n2 = GetAssetTools()
   > Set AssetTools(AssetTools:Asset Tools Interface=@n2) -> (Output_Get:Asset Tools Interface)
   n3 = GetDisplayName(Object=@n1)
   n4 = Append!(A=Get CurrentSelectionPath(), B=/, C:String=@n3) -> (ReturnValue)
   > n5 = LoadAsset(AssetPath=@n4)
   > Set SelectedAssets(SelectedAssets:Array of Object References=Make Array("[0]":Object Reference=@n5) -> (Array:Array of Object References)) -> (Output_Get:Array of Object References)
   n6 = FindSubstring(SearchIn=@n3, Substring=_)
   > Return Node!(
    execute
    CurrentIDName:String=@n3
    CurrentVehicleName:String=RightChop!(SourceString:String=@n3, Count:int=int + int(A=@n6)) -> (ReturnValue:String)
    CurrentPathAndName:String=@n4
   )
  ]}
 GetSafeTextForLocalization {nodes [
   n1 = GetSafeTextForLocalization!() -> (then, Asset:Object Reference, ProperyName:String, TextToLocalize:String) {is=title K2Node_FunctionEntry}
   n2 = GetObjectName(Object=@n1)
   > Return Node!(
    execute
    SafeText:Text="To Text (String)"(InString=Append!(A=@n2, B=_, C=@n1.ProperyName, D:String=_, E:String=@n1.TextToLocalize) -> (ReturnValue))
   )
  ]}
 GetVariantChannelName {nodes [
   n1 = GetVariantChannelName!() -> (then, InFortID:Fort Item Definition Object Reference) {is=title K2Node_FunctionEntry}
   > Set FortID(FortID=@n1)
   > n2 = Switch on DelMar_ProductTypeEnum(Selection=Get ProductTypeSelection() -> (ProductTypeSelection=NewEnumerator0))
   n3 = GetVehicleNameFromFortID!(
    execute=[@n2.NewEnumerator0, @n2.NewEnumerator1, @n2.NewEnumerator2, @n2.NewEnumerator4, @n2.NewEnumerator5, @n2.NewEnumerator6]
    self
    FortID:Fort Item Definition Object Reference
   ) -> (then, VehicleName:String)
   n4 = Get ProductTypeSelection() -> (ProductTypeSelection=NewEnumerator0)
   > Return Node!(execute, VariantChannelName:String=Append!(A=@n3, B=" ", C=Enum to String(Enumerator=@n4), D=" Color") -> (ReturnValue))
   Return Node(execute=@n2.NewEnumerator3, VariantChannelName=Decal Color)
  ]}
 GetVariantChannelTagForRichColorComponent {nodes [
   n1 = GetSelectedOption!(self:"Combo Box (String) Object Reference"=Get ProductTypeSelection() -> (ProductTypeSelection=NewEnumerator0)) -> (ReturnValue:String)
   Append(B=" ", C=@n1, D=" Color")
   Append!(A=Cosmetics.Variant.Channel.Vehicle., B=@n1, C=.Color) -> (ReturnValue)
   GetVariantChannelTagForRichColorComponent() {is=title K2Node_FunctionEntry}
  ]}
 GetVehicleCosmeticIDClass {nodes [
   GetVehicleCosmeticIDClass() {is=title K2Node_FunctionEntry}
   > n1 = Switch on DelMar_ProductTypeEnum(Selection=Get ProductTypeSelection() -> (ProductTypeSelection=NewEnumerator0))
   > Return Node!(execute, NewParam:Vehicle Cosmetics Item Definition Soft Class Reference=&VehicleCosmeticsItemDefinition_Body)
   Return Node(execute=@n1.NewEnumerator1, NewParam=&VehicleCosmeticsItemDefinition_DriftTrail)
   Return Node(execute=@n1.NewEnumerator2, NewParam=&VehicleCosmeticsItemDefinition_Booster)
   Return Node(execute=@n1.NewEnumerator3, NewParam=&VehicleCosmeticsItemDefinition_Skin)
   Return Node(execute=@n1.NewEnumerator4, NewParam=/Script/VehicleCosmeticsRuntime.VehicleCosmeticsItemDefinition_Zest)
   Return Node(execute=@n1.NewEnumerator5, NewParam=/Script/VehicleCosmeticsRuntime.VehicleCosmeticsItemDefinition_Macaron)
   Return Node(execute=@n1.NewEnumerator6, NewParam=&VehicleCosmeticsItemDefinition_Wheel)
  ]}
 GetVehicleIDClass {nodes [
   GetVehicleIDClass() {is=title K2Node_FunctionEntry}
   > n1 = Switch on DelMar_ProductTypeEnum(Selection=Get ProductTypeSelection() -> (ProductTypeSelection=NewEnumerator0))
   > Return Node(NewParam:Fort Vehicle Cosmetics Item Definition Class Reference=/Script/VehicleCosmeticsRuntime.FortVehicleCosmeticsItemDefinition_Body)
   Return Node(execute=@n1.NewEnumerator1, NewParam=/Script/VehicleCosmeticsRuntime.FortVehicleCosmeticsItemDefinition_DriftTrail)
   Return Node(execute=@n1.NewEnumerator2, NewParam=/Script/VehicleCosmeticsRuntime.FortVehicleCosmeticsItemDefinition_Booster)
   Return Node(execute=@n1.NewEnumerator3, NewParam=/Script/VehicleCosmeticsRuntime.FortVehicleCosmeticsItemDefinition_Skin)
   Return Node(execute=@n1.NewEnumerator4, NewParam=/Script/VehicleCosmeticsRuntime.FortVehicleCosmeticsItemDefinition_Zest)
   Return Node(execute=@n1.NewEnumerator5, NewParam=/Script/VehicleCosmeticsRuntime.FortVehicleCosmeticsItemDefinition_Macaron)
   Return Node(execute=@n1.NewEnumerator6, NewParam=/Script/VehicleCosmeticsRuntime.FortVehicleCosmeticsItemDefinition_Wheel)
  ]}
 GetVehicleNameFromFortID {nodes [
   n1 = GetVehicleNameFromFortID!() -> (then, FortID:Fort Item Definition Object Reference) {is=title K2Node_FunctionEntry}
   n2 = "To String (Text)"(InText=GetDisplayName!(self:Item Definition Base Object Reference=@n1) -> (ReturnValue:Text))
   n3 = int + int(A=FindSubstring(SearchIn=@n2, Substring=_))
   > Return Node!(execute, VehicleName:String=Replace(SourceString=RightChop(SourceString=@n2, Count=@n3), From=")"))
  ]}
 HandleVariableInitialization {nodes [
   HandleVariableInitialization() {is=title K2Node_FunctionEntry}
   > Clear(TargetArray:Array of Object References=Get SelectedAssets() -> (SelectedAssets:Array of Object References))
   > Clear(TargetArray:Array of Names=Get PaintNames())
   > Clear(TargetArray:Array of Fort Item Definition Object References=Get AllFortIDs())
   > Set VehicleName()
   > Set SelectedIDName(SelectedIDName:String) -> (Output_Get:String)
   > Set CurrentSelectionPath()
   > Set SelectedID_PathAndName(SelectedID_PathAndName:String) -> (Output_Get:String)
  ]}
 MakeInstanceTableRow {nodes [
   n1 = MakeInstanceTableRow!() -> (then, RowName:Name) {is=title K2Node_FunctionEntry}
   n2 = String To Name(InString=Replace(SourceString="To String (Name)"(InName=@n1), From=" "))
   n3 = Get VehicleCosmeticDataTable()
   > n4 = MakeInstancedStruct!(
    execute:Exec
    Value:Cosmetic Property Table Row Structure=Make Cosmetic Property Table Row!(
     TableRow:Data Table Row Handle Structure=Make Data Table Row Handle!(DataTable:Data Table Object Reference=@n3, RowName:Name=@n2) -> (DataTableRowHandle:Data Table Row Handle Structure) {is=title K2Node_MakeStruct}
     PropertyTag:Gameplay Tag Structure=MakeLiteralGameplayTag!(Value:Gameplay Tag Structure=(TagName="CosmeticProperty.Vehicle.PaintedRow")) -> (ReturnValue:Gameplay Tag Structure)
    ) -> (CosmeticProperty_TableRow:Cosmetic Property Table Row Structure) {is=title K2Node_MakeStruct}
   ) -> (then:Exec, ReturnValue:Instanced Struct Structure) {is=title K2Node_InstancedStruct}
   > Return Node!(execute, InstanceTableRow:Instanced Struct Structure=@n4)
  ]}
 PrintDebugData {nodes [
   PrintDebugData() {is=title K2Node_FunctionEntry}
   > HandleVariableInitialization() {is=title K2Node_CallFunction}
   > GetSelectedAssetAndInitVars() {is=title K2Node_CallFunction}
   > PrintString(InString=Append!(A="Current Selection is: ", B=Get SelectedIDName() -> (SelectedIDName:String)) -> (ReturnValue))
   > PrintString(InString=Append(A="Path of Selection: ", B=Get CurrentSelectionPath()))
   > PrintString(InString=Append(A="Path and File in Variable: ", B=Get SelectedID_PathAndName() -> (SelectedID_PathAndName:String)))
   > PrintString(InString=Append(A="Vehicle Name From File: ", B=Get VehicleName()), TextColor=(R=0,G=0.66,B=0.126,A=1), Duration=5)
   > n1 = GetDataTableRowNames(Table=&DT_VehicleCosmetics_Painted)
   > n2 = For Each Loop(Array=@n1)
   > PrintString(InString="To String (Name)"(InName=@n2))
  ]}
 SetDebugStrings {nodes [
   n1 = SetDebugStrings!() -> (then, Fort ID:Fort Item Definition Object Reference) {is=title K2Node_FunctionEntry}
   n2 = GetDisplayName!(Object:Object Reference=@n1) -> (ReturnValue:String)
   > n3 = Set SelectedIDName(SelectedIDName=@n2)
   > Set SelectedID_PathAndName(SelectedID_PathAndName=Append!(A=GetCurrentContentBrowserPath(), B=/, C:String=@n2) -> (ReturnValue))
   > Set VehicleName(VehicleName=RightChop(SourceString=@n3, Count=8))
  ]}
 SetInitialFortIDProperties {nodes [
   n1 = SetInitialFortIDProperties!() -> (then, FortID:Fort Item Definition Object Reference, VCID:Vehicle Cosmetics Item Definition Object Reference) {is=title K2Node_FunctionEntry}
   > n2 = GetGameplayTagFromProductType!(execute, self) -> (then, GameplayTag:Gameplay Tag Structure) {is=title K2Node_CallFunction}
   > Execute Python Script!(
    execute
    PythonScript="import unreal\r\n\r\nfort_id_asset.set_editor_property(\r\n                \"ItemTypeTag\",\r\n                variant_channel_tag)\r\n\r\nfort_id_asset.set_editor_property(\r\n                \"VehicleCosmeticsItemDef\",\r\n                vcid_asset)\r\n\r\nfort_id_asset.set_editor_property(\r\n                \"Rarity\",\r\n                rarity)\r\n\r\nfrontend_component = unreal.FortItemComponent_FrontendPreview(fort_id_asset)\r\n\r\ncomponents_container = fort_id_asset.get_editor_property(\"ComponentContainer\")\r\n\r\ncomponents_array = components_container.get_editor_property(\"components\")\r\n\r\ncomponents_array.pop(0)\r\ncomponents_array.append(frontend_component)\r\n"
    FortIDAsset:Fort Item Definition Object Reference=@n1
    VCIDAsset:Vehicle Cosmetics Item Definition Object Reference=@n1.VCID
    VariantChannelTag:Gameplay Tag Structure=@n2
    rarity:EFortRarity Enum=Get RaritySelection() -> (RaritySelection=Common)
   ) -> (then, ReturnValue) {inputs [FortIDAsset, VCIDAsset, VariantChannelTag, rarity]}
  ]}
 SetInitialVehicleCosmeticProperties {nodes [
   n1 = SetInitialVehicleCosmeticProperties!() -> (then, FortID:Fort Item Definition Object Reference, VehicleCosmetic:Vehicle Cosmetics Item Definition Object Reference) {is=title K2Node_FunctionEntry}
   > Branch(Condition=Is Not Empty(TargetArray=Get PaintNames()))
   > n2 = GetGameplayTagFromProductType()
   > n3 = GetVehicleNameFromFortID(FortID=@n1) {is=title K2Node_CallFunction}
   > Execute Python Script!(
    execute
    PythonScript="import unreal\r\n\r\nvariants = fort_id_asset.get_editor_property(\"ItemVariants\")\r\nnew_variant = unreal.FortCosmeticPropertyVariant(fort_id_asset)\r\n\r\ntagChannel = unreal.FortniteEditorLibrary.request_gameplay_tag(\"Cosmetics.Variant.Channel.Vehicle.Painted\", True)\r\n\r\nnew_variant.set_editor_property(\r\n                \"VariantChannelTag\",\r\n                tagChannel)\r\n\r\nloc_channel_name = unreal.NSLOCTEXT('', fort_id_asset.get_name() + \".\" + new_variant.get_name(), variant_channel_name)\r\n\r\nnew_variant.set_editor_property(\r\n                \"VariantChannelName\",\r\n                loc_channel_name)\r\n\r\nvariants.append(new_variant)\r\n"
    FortIDAsset=@n1
    VCIDAsset=@n1.VehicleCosmetic
    ProductTypeTag:Gameplay Tag Structure=@n2
    VariantChannelName:String=Append!(A=@n3, B=" Painted Style") -> (ReturnValue)
    FortIDName:Text="To Text (String)"(InString=GetObjectName(Object=@n1))
    ProductTypeName:String=Enum to String(Enumerator=Get ProductTypeSelection() -> (ProductTypeSelection=NewEnumerator0))
   ) -> (then, ReturnValue) {inputs [FortIDAsset, VCIDAsset, ProductTypeTag, VariantChannelName, FortIDName, ProductTypeName]}
  ]}
 SetVariantProperties {nodes [
   n1 = SetVariantProperties!() -> (
    then
    PaintName:Name
    FortID:Fort Item Definition Object Reference
    Variant:Fort Vehicle Cosmetics Variant Token Type Object Reference
    VCIDAsset:Vehicle Cosmetics Item Definition Object Reference
   ) {is=title K2Node_FunctionEntry}
   > n2 = MakeInstanceTableRow!(execute, self, RowName:Name=@n1) -> (then, InstanceTableRow:Instanced Struct Structure) {is=title K2Node_CallFunction}
   > n3 = GetPreviewTextureForColor!(execute, self, PaintName:Name=@n1) -> (then, FoundTexture:Texture Object Reference) {is=title K2Node_CallFunction}
   > n4 = GetGameplayTagFromPaintName!(execute, self, PaintName:Name=@n1) -> (then, GameplayTag:Gameplay Tag Structure) {is=title K2Node_CallFunction}
   > n5 = GetPaintedRowTag!(execute, self) -> (then, GameplayTag:Gameplay Tag Structure) {is=title K2Node_CallFunction}
   > Execute Python Script!(
    execute
    PythonScript="import unreal\r\n\r\nvariants = fort_id_asset.get_editor_property(\"ItemVariants\")\r\nnew_variant = variants[0]\r\n\r\ncosmetic_variant = new_variant.get_editor_property(\"GenericPropertyOptions\")\r\n\r\ncosmetic_prop_def = unreal.CosmeticPropertyVariantDef()\r\n\r\ncosmetic_prop_array = cosmetic_prop_def.get_editor_property(\"CosmeticProperties\")\r\n\r\ncosmetic_prop_array.append(cosmetic_table_row)\r\n\r\nif is_default_object == True:\r\n    loc_variant_name = unreal.NSLOCTEXT('', fort_id_asset.get_name() + \".none\", paint_name)\r\nelse:\r\n    loc_variant_name = unreal.NSLOCTEXT('', fort_id_asset.get_name() + \".\" + current_variant.get_name(), paint_name)\r\n\r\ncosmetic_prop_def.set_editor_property(\r\n                \"VariantName\",\r\n                loc_variant_name)\r\n\r\ncosmetic_prop_def.set_editor_property(\r\n                \"CustomizationVariantTag\",\r\n                cosmetic_tag)\r\n\r\ncosmetic_prop_def.set_editor_property(\r\n                \"start_unlocked\",\r\n                False)\r\n\r\ncosmetic_prop_def.set_editor_property(\r\n                \"hide_if_not_owned\",\r\n                True)\r\n\r\ncosmetic_prop_def.set_editor_property(\r\n                \"PreviewImage\",\r\n                preview_texture)\r\n\r\nif is_default_object == True:\r\n    cosmetic_prop_def.set_editor_property(\r\n                \"start_unlocked\",\r\n                True)\r\n\r\n    cosmetic_prop_def.set_editor_property(\r\n                \"is_default\",\r\n                True)\r\n\r\n    cosmetic_prop_def.set_editor_property(\r\n                \"hide_if_not_owned\",\r\n                True)\r\n\r\ncosmetic_variant.append(cosmetic_prop_def)"
    CurrentVariant:Fort Vehicle Cosmetics Variant Token Type Object Reference=@n1.Variant
    FortIDAsset=@n1.FortID
    PaintName:String="To String (Name)"(InName=@n1)
    CosmeticTag:Gameplay Tag Structure=@n4
    PaintedRowTag:Gameplay Tag Structure=@n5
    VehicleCosmeticDataTable:Data Table Object Reference=Get VehicleCosmeticDataTable()
    bIsDefaultObject:bool="Equal (Name)"(A=@n1)
    CosmeticTableRow:Instanced Struct Structure=@n2
    PreviewTexture:Texture Object Reference=@n3
    ProductTypeName=Enum to String(Enumerator=Get ProductTypeSelection() -> (ProductTypeSelection=NewEnumerator0))
   ) -> (then, ReturnValue) {inputs [
     CurrentVariant
     FortIDAsset
     PaintName
     CosmeticTag
     PaintedRowTag
     VehicleCosmeticDataTable
     bIsDefaultObject
     CosmeticTableRow
     PreviewTexture
     ProductTypeName
    ]}
  ]}
 SetupBodyVCID {nodes [
   n1 = SetupBodyVCID!() -> (then, VCID:Vehicle Cosmetics Item Definition Object Reference, AssetName:String) {is=title K2Node_FunctionEntry}
   > n2 = LoadAsset(AssetPath=/VehicleCosmetics/Mutable/_Root/CO_VehicleCosmeticsRoot)
   > Execute Python Script!(
    execute
    PythonScript="import unreal\r\n\r\ncheat_array = vcid.get_editor_property(\"CheatNames\")\r\n\r\ncheat_array.append(asset_name)\r\n\r\nvcid.set_editor_property(\r\n                \"CustomizableObject\",\r\n                customization_object)\r\n\r\nbody_group = vcid.get_editor_property(product_type + \"Group\")\r\n\r\nbody_group.set_editor_property(\r\n                \"ParameterName\",\r\n                product_type)\r\n\r\nbody_group.set_editor_property(\r\n                \"ParameterValue\",\r\n                asset_name + product_type)\r\n\r\nvcid.set_editor_property(\r\n                \"SkinDataTableParameterName\",\r\n                asset_name + \"Skins\")\r\n\r\nvcid.set_editor_property(\r\n                \"BodyColorParameterName\",\r\n                product_type + \"Color\")\r\n\r\nvcid.set_editor_property(\r\n                \"SkinColorParameterName\",\r\n                \"SkinColor\")\r\n\r\nvcid.set_editor_property(\r\n                \"PaintedDataTableParameterName\",\r\n                product_type + \"Painted\")"
    CustomizationObject:Object Reference=@n2
    ProductType:String=Enum to String(Enumerator=Get ProductTypeSelection() -> (ProductTypeSelection=NewEnumerator0))
    AssetName:String=@n1.AssetName
    PaintedDataTableName:String=Append(B=Painted)
    VCID:Vehicle Cosmetics Item Definition Object Reference=@n1
   ) -> (then, ReturnValue) {inputs [CustomizationObject, ProductType, AssetName, PaintedDataTableName, VCID]}
  ]}
 SetupBoosterVCID {nodes [
   n1 = SetupBoosterVCID!() -> (then, VCID:Vehicle Cosmetics Item Definition Object Reference, AssetName:String) {is=title K2Node_FunctionEntry}
   > Execute Python Script!(
    execute
    PythonScript="import unreal\r\n\r\ncheat_array = vcid.get_editor_property(\"CheatNames\")\r\n\r\ncheat_array.append(asset_name)\r\n\r\nvcid.set_editor_property(\r\n                \"CheatLoadoutSlotName\",\r\n                \"CosmeticLoadoutSlotTemplate:LoadoutSlot_Vehicle_Booster\")\r\n"
    AssetName=@n1.AssetName
    VCID=@n1
   ) -> (then, ReturnValue) {inputs [AssetName,VCID]}
  ]}
 SetupDriftTrail {nodes [
   n1 = SetupDriftTrail!() -> (then, VCID:Vehicle Cosmetics Item Definition Object Reference, AssetName:String) {is=title K2Node_FunctionEntry}
   > Execute Python Script(
    PythonScript="import unreal\r\n\r\ncheat_array = vcid.get_editor_property(\"CheatNames\")\r\n\r\ncheat_array.append(asset_name)\r\n\r\nvcid.set_editor_property(\r\n                \"CheatLoadoutSlotName\",\r\n                \"CosmeticLoadoutSlotTemplate:LoadoutSlot_Vehicle_DriftSmoke\")\r\n"
    AssetName=@n1.AssetName
    VCID=@n1
   ) {inputs [AssetName,VCID]}
  ]}
 SetupSkinVCID {nodes [
   n1 = SetupSkinVCID!() -> (then, VCID:Vehicle Cosmetics Item Definition Object Reference, AssetName:String) {is=title K2Node_FunctionEntry}
   > Execute Python Script!(
    execute
    PythonScript="import unreal\r\n\r\ncheat_array = vcid.get_editor_property(\"CheatNames\")\r\n\r\ncheat_array.append(asset_name)\r\n\r\nvcid.set_editor_property(\r\n                \"CheatloadoutSlotName\",\r\n                \"CosmeticLoadoutSlotTemplate:LoadoutSlot_Vehicle_Skin\")\r\n\r\nvcid.set_editor_property(\r\n                \"PaintedDataTableParameterName\",\r\n                \"SkinPainted\")\r\n\r\nvcid.set_editor_property(\r\n                \"PaintedEnumParameterName\",\r\n                \"SkinPaintType\")\r\n"
    VCID=@n1
    AssetName=@n1.AssetName
   ) -> (then, ReturnValue) {inputs [VCID,AssetName]}
  ]}
 SetupWheelVCID {nodes [
   n1 = SetupWheelVCID!() -> (then, VCID:Vehicle Cosmetics Item Definition Object Reference, AssetName:String) {is=title K2Node_FunctionEntry}
   > n2 = LoadAsset(AssetPath=&CO_VehicleCosmeticsRoot)
   > Execute Python Script!(
    execute
    PythonScript="import unreal\r\n\r\ncheat_array = vcid.get_editor_property(\"CheatNames\")\r\n\r\ncheat_array.append(asset_name)\r\n\r\nvcid.set_editor_property(\r\n                \"ComponentIndex\",\r\n                1)\r\n\r\nvcid.set_editor_property(\r\n                \"CheatLoadoutSlotName\",\r\n                \"CosmeticLoadoutSlotTemplate:LoadoutSlot_Vehicle_Wheel\")\r\n\r\nvcid.set_editor_property(\r\n                \"CustomizableObject\",\r\n                customization_object)\r\n\r\nwheel_group = vcid.get_editor_property(\"WheelGroup\")\r\n\r\nwheel_group.set_editor_property(\r\n                \"ParameterName\",\r\n                \"Wheels\")\r\n\r\nwheel_group.set_editor_property(\r\n                \"ParameterValue\",\r\n                \"Wheel_\" + asset_name)\r\n\r\n\r\nvcid.set_editor_property(\r\n                \"PaintedDataTableParameterName\",\r\n                \"WheelPainted\")\r\n\r\n"
    CustomizationObject:Object Reference=@n2
    AssetName=@n1.AssetName
    VCID=@n1
   ) -> (then, ReturnValue) {inputs [CustomizationObject, AssetName, VCID]}
  ]}
 TestFunction {nodes [
   n1 = TestFunction!() -> (then, Type:int) {is=title K2Node_FunctionEntry}
   > n2 = Switch on Int!(execute:Exec, Selection:int=@n1) -> (Default:Exec, 0:Exec, 1:Exec, 2:Exec, 3:Exec, 4:Exec, 5:Exec, 6:Exec)
   Set ProductTypeSelection(execute=@n2.0, ProductTypeSelection:DelMar_ProductTypeEnum Enum=NewEnumerator0) -> (Output_Get:DelMar_ProductTypeEnum Enum=NewEnumerator0)
   Set ProductTypeSelection(execute=@n2.1, ProductTypeSelection=NewEnumerator1) -> (Output_Get=NewEnumerator0)
   Set ProductTypeSelection(execute=@n2.2, ProductTypeSelection=NewEnumerator2) -> (Output_Get=NewEnumerator0)
   Set ProductTypeSelection(execute=@n2.3, ProductTypeSelection=NewEnumerator3) -> (Output_Get=NewEnumerator0)
   Set ProductTypeSelection(execute=@n2.4, ProductTypeSelection=NewEnumerator4) -> (Output_Get=NewEnumerator0)
   Set ProductTypeSelection(execute=@n2.5, ProductTypeSelection=NewEnumerator5) -> (Output_Get=NewEnumerator0)
   Set ProductTypeSelection(execute=@n2.6, ProductTypeSelection=NewEnumerator6) -> (Output_Get=NewEnumerator0)
   PrintString(InString="I was called!")
  ]}
]
variables [
 AllFortIDs {type="array<FortItemDefinition>", default []}
 AllVCIDs {type="array<VehicleCosmeticsItemDefinition>", default []}
 AntennaIsFlagSubclass {type=boolean}
 AssetTools {default=None}
 CreatedFortVariants {type="array<FortVehicleCosmeticsVariantTokenType>", default []}
 CreateNewFortIDs {type="array<string>", default []}
 CurrentSelectionPath {type=string, default=""}
 FortRelease {type=FortReleaseVersion, default {versionName=Unset}}
 OverwriteExistingID {type=FortItemDefinition, default=null}
 PaintNames {type="array<string>", default []}
 Parent Name {type=string, default=""}
 ProductTypeSelection {type=DelMar_ProductTypeEnum, default=Body}
 RaritySelection {type=EFortRarity, default=Common}
 SelectedAssets {type="array<Object>", default [null]}
 SelectedID_PathAndName {type=string, default=""}
 SelectedIDName {type=string, default=""}
 VehicleCosmeticDataTable {type=DataTable, default=null}
 VehicleName {type=string, default=VEHICLENAME}
 VTID_Template {type=FortVehicleCosmeticsItemDefinition_Body, default=null}
]
widgets {
 info {rootWidgetClass=&VerticalBox, widgetCount=20}
 widgets [
  VerticalBox_82 {widgetClassPath=&VerticalBox, widget=VerticalBox_82}
  ScaleBox_163 {widgetClassPath=&ScaleBox, parent=VerticalBox_82, widget=ScaleBox_163, slot {class=VerticalBoxSlot}}
  VerticalBox_117 {
   widgetClassPath=&VerticalBox
   parent=ScaleBox_163
   widget=VerticalBox_117
   slot {class=ScaleBoxSlot, properties {horizontalAlignment=HAlign_Fill, verticalAlignment=VAlign_Fill}}
  }
  HorizontalBox_145 {
   widgetClassPath=&HorizontalBox
   parent=VerticalBox_117
   widget=HorizontalBox_145
   slot {class=VerticalBoxSlot, properties {horizontalAlignment=HAlign_Center, padding=<ltrb 10 10 10 10>}}
  }
  IMG_RL_Logo {
   widgetClassPath=&Image
   bIsVariable=true
   parent=HorizontalBox_145
   widget=IMG_RL_Logo
   properties {brush {
     bIsDynamicallyLoaded=false
     drawAs=Image
     imageType=NoImage
     mirroring=NoMirror
     resourceName=None
     resourceObject=/Game/NeverCook/TechArt/VehicleCosmeticsTools/VehicleCosmetics_Picker/Icons/UI_RL_Icon
     tiling=NoTile
     imageSize=<64 64>
     margin=<ltrb 0 0 0 0>
     outlineSettings {
      bUseBrushTransparency=false
      roundingType=HalfHeightRadius
      width=0
      color {colorUseRule=UseColor_Specified, specifiedColor=<rgba 0 0 0 0>}
      cornerRadii=<0 0 0 0>
     }
     tintColor {colorUseRule=UseColor_Specified, specifiedColor=<rgba 1 1 1 1>}
     uVRegion {bIsValid=false, max=<0 0>, min=<0 0>}
    }}
   slot {class=HorizontalBoxSlot, properties {horizontalAlignment=HAlign_Left, verticalAlignment=VAlign_Center}}
  }
  Spacer_55 {widgetClassPath=&Spacer, parent=HorizontalBox_145, widget=Spacer_55, properties {size=<15 1>}, slot {class=HorizontalBoxSlot}}
  TXT_HeaderTitle {
   widgetClassPath=&TextBlock
   parent=HorizontalBox_145
   widget=TXT_HeaderTitle
   properties {
    text=Fort ID Tool
    font {
     bForceMonospaced=false
     bMaterialIsStencil=false
     fontMaterial=null
     fontObject=null
     letterSpacing=0
     monospacedWidth=1
     size=48
     skewAmount=0
     typefaceFontName=Default
     outlineSettings {
      bApplyOutlineToDropShadows=false
      bMiteredCorners=false
      bSeparateFillAlpha=false
      outlineMaterial=null
      outlineSize=0
      outlineColor=<rgba 0 0 0 1>
     }
    }
   }
   slot {class=HorizontalBoxSlot, properties {horizontalAlignment=HAlign_Left, verticalAlignment=VAlign_Center}}
  }
  HBox_VehicleName {
   widgetClassPath=&HorizontalBox
   parent=VerticalBox_117
   widget=HBox_VehicleName
   slot {class=VerticalBoxSlot, properties {horizontalAlignment=HAlign_Center, padding=<ltrb 10 0 10 0>}}
  }
  VerticalBox_145 {widgetClassPath=&VerticalBox, parent=HBox_VehicleName, widget=VerticalBox_145, slot {class=HorizontalBoxSlot}}
  SizeBox_68 {widgetClassPath=&SizeBox, parent=VerticalBox_145, widget=SizeBox_68, slot {class=VerticalBoxSlot}}
  FortIDInputDetailView {
   widgetClassPath=&DetailsView
   parent=SizeBox_68
   widget=FortIDInputDetailView
   properties {categoriesToShow [FortIDInput]}
   slot {class=SizeBoxSlot}
  }
  SizeBox_228 {widgetClassPath=&SizeBox, parent=VerticalBox_145, widget=SizeBox_228, slot {class=VerticalBoxSlot}}
  ProductTypeDetailView {
   widgetClassPath=&DetailsView
   parent=SizeBox_228
   widget=ProductTypeDetailView
   properties {categoriesToShow [ProductType]}
   slot {class=SizeBoxSlot}
  }
  SizeBox {widgetClassPath=&SizeBox, parent=VerticalBox_145, widget=SizeBox, slot {class=VerticalBoxSlot}}
  ParentNameDetailView {
   widgetClassPath=&DetailsView
   parent=SizeBox
   widget=ParentNameDetailView
   properties {categoriesToShow [Additional Settings]}
   slot {class=SizeBoxSlot}
  }
  SizeBox_158 {widgetClassPath=&SizeBox, parent=VerticalBox_145, widget=SizeBox_158, slot {class=VerticalBoxSlot}}
  RarityDetailView {
   widgetClassPath=&DetailsView
   parent=SizeBox_158
   widget=RarityDetailView
   properties {categoriesToShow [Rarity]}
   slot {class=SizeBoxSlot}
  }
  Button_CreatePaintedVariants {
   widgetClassPath=&EditorUtilityButton
   parent=VerticalBox_145
   widget=Button_CreatePaintedVariants
   slot {class=VerticalBoxSlot, properties {padding=<ltrb 10 24 10 10>}}
  }
  TXT_CreatePaintedVariants {
   widgetClassPath=&TextBlock
   parent=Button_CreatePaintedVariants
   widget=TXT_CreatePaintedVariants
   properties {
    text=Create Variants
    colorAndOpacity {colorUseRule=UseColor_Specified, specifiedColor=<rgba 0.526 0.526 0.526 1>}
    font {
     bForceMonospaced=false
     bMaterialIsStencil=false
     fontMaterial=null
     fontObject=null
     letterSpacing=0
     monospacedWidth=1
     size=24
     skewAmount=0
     typefaceFontName=RLSmall-Regular
     outlineSettings {
      bApplyOutlineToDropShadows=false
      bMiteredCorners=false
      bSeparateFillAlpha=false
      outlineMaterial=null
      outlineSize=0
      outlineColor=<rgba 0 0 0 1>
     }
    }
   }
   slot {class=ButtonSlot}
  }
  ScaleBox_135 {
   widgetClassPath=&ScaleBox
   parent=HBox_VehicleName
   widget=ScaleBox_135
   slot {class=HorizontalBoxSlot, properties {padding=<ltrb 8 0 8 0>}}
  }
 ]
}
