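; Editor utility task that batch-generates painted variants for vehicle cosmetic Fort item definitions.
; For each Fort ID, it creates a variant token asset per paint name, then adds matching options with table row, tag and preview icon.
; Body and Skin types also get a rich color variant.
; Embedded Python performs property edits, while Delay-stepped, cancellable loops report progress and save the item definitions and variant assets.
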(event Task|EventBeginExecution
  (bind _allfortindex (Variables|Default|GetAllFortIndex))
  (bind _allfortids (Variables|Default|GetAllFortIDs))
  (bind _output (Utilities|Array|Get(acopy) _allfortids _allfortindex))
  (bind _returnvalue (EditorSubsystems|GetEditorUtilitySubsystem))
  (bind _savingindex (Variables|Default|GetSavingIndex))
  (bind _allcreatedfortvariants (Variables|Default|GetAllCreatedFortVariants))
  (Task|SetTaskNotificationText (Utilities|Text|MakeLiteralText "INVTEXT(\"Starting..\")"))
  (Utilities|Array|Clear (Variables|Default|GetAllCreatedFortVariants))
  (Variables|Default|SetSavingIndex 0)
  (Variables|Default|SetCreateVariantIndex 0)
  (Variables|Default|SetAllFortIndex 0)
  (Utilities|FlowControl|Delay 0 0.1)
  (if (< (Variables|Default|GetAllFortIndex) (Utilities|Array|Length (Variables|Default|GetAllFortIDs)))
    (if (Task|WasCancelRequested)
      _
      (elif (< (Variables|Default|GetCreateVariantIndex) (Utilities|Array|Length (Variables|Default|GetPaintNames)))
        (if (not (Task|WasCancelRequested))
          (Default|GenerateVariantsforFortID (Utilities|Array|Get(acopy) (Variables|Default|GetAllFortIDs) (Variables|Default|GetAllFortIndex)) (Variables|Default|GetCreateVariantIndex))
          (Math|Integer|IncrementInt (Variables|Default|GetCreateVariantIndex)
            (:%20%20
              (Utilities|FlowControl|Delay 0 0.3))))
        (else
          (Default|AddCreatedVariantstoFortID _output (Variables|Default|GetCreatedFortVariants) (Utilities|Array|Get(acopy) (Variables|Default|GetAllVCIDs) (Variables|Default|GetAllFortIndex)))
          (Task|SetTaskNotificationText (Utilities|Text|FormatText "INVTEXT(\"Set Variants Properties\")"))
          (Utilities|FlowControl|Delay 0 0.3)
          (Default|CreateRichColorVariantProperty _output)
          (Utilities|FlowControl|Delay 0 0.3)
          (Task|SetTaskNotificationText (Utilities|Text|FormatText "INVTEXT(\"Saving Fort ID Asset\")"))
          (EditorScripting|Asset|SaveLoadedAsset _output :!class /Script/EditorScriptingUtilities.EditorAssetLibrary)
          (Utilities|FlowControl|Delay 0)
          (Math|Integer|IncrementInt (Variables|Default|GetAllFortIndex)
            (:%20%20
              _)))))
    (bind _returnvalue_2 (Development|Editor|GetTabIDFromBlueprint _returnvalue "/Game/NeverCook/TechArt/VariantTool/BPW_FortIDTool.BPW_FortIDTool"))
    (Development|Editor|UnregisterTabbyID _returnvalue _returnvalue_2)
    (Task|FinishExecutingTask true))
  (if (< (Variables|Default|GetSavingIndex) (Utilities|Array|Length (Variables|Default|GetAllCreatedFortVariants)))
    (if (not (Task|WasCancelRequested))
      (EditorScripting|Asset|SaveLoadedAsset (Utilities|Array|Get(acopy) _allcreatedfortvariants _savingindex) :!class /Script/EditorScriptingUtilities.EditorAssetLibrary)
      (Task|SetTaskNotificationText (Utilities|Text|FormatText "INVTEXT(\"{step}/{max} - Saving Variant Colors\")" (+ _savingindex 1) (Utilities|Array|Length _allcreatedfortvariants)))
      (Math|Integer|IncrementInt (Variables|Default|GetSavingIndex)
        (:%20%20
          (Utilities|FlowControl|Delay 0))))))

(fn AddCreatedVariantsToFortID
  (params (Fort%20ID :type Fort%20Item%20Definition%20Object%20Reference) (Created%20Variants :type Array%20of%20Fort%20Vehicle%20Cosmetics%20Variant%20Token%20Type%20Object%20References) (VCIDAsset :type Vehicle%20Cosmetics%20Item%20Definition%20Object%20Reference))
  (LocalVariables|SetLocalVCIDAsset VCIDAsset)
  (Default|SetInitialVehicleCosmeticProperties Fort%20ID VCIDAsset)
  ;[1 1 1] Adding Variants to the Create Variants Array
  (for (_array_element _array_index) Created%20Variants
    (Default|SetVariantProperties (Utilities|Array|Get(aref) (Variables|Default|GetPaintNames) _array_index) Fort%20ID _array_element (LocalVariables|GetLocalVCIDAsset))
    (Task|SetTaskNotificationText (Utilities|Text|FormatText "INVTEXT(\"{step}/{max} -Set Variants Properties\")" (+ _array_index 1) (Utilities|Array|Length Created%20Variants)))))

(fn CreateRichColorVariantProperty
  (params (FortID :type Fort%20Item%20Definition%20Object%20Reference))
  (if (or (Utilities|String|Contains (Utilities|String|EnumtoString (Variables|Default|GetProductTypeSelection)) "Body") (Utilities|String|Contains (Utilities|String|EnumtoString (Variables|Default|GetProductTypeSelection)) "Skin"))
    (bind _variantchannelname (Default|GetVariantChannelName FortID))
    (bind _returnvalue (EditorScripting|Asset|LoadAsset "/VehicleCosmetics/Mutable/Bodies/CS_Bodies_PrimaryColor.CS_Bodies_PrimaryColor" :!class /Script/EditorScriptingUtilities.EditorAssetLibrary))
    (Python|Execution|ExecutePythonScript "import unreal

variants = fort_id_asset.get_editor_property(\"ItemVariants\")
new_variant = unreal.FortCosmeticRichColorVariant(fort_id_asset)

loc_channel_name = unreal.NSLOCTEXT('', fort_id_asset.get_name() + \".\" + new_variant.get_name(), variant_channel_name)

new_variant.set_editor_property(
                \"VariantChannelName\",
                loc_channel_name)

tagChannel = unreal.FortniteEditorLibrary.request_gameplay_tag(variant_channel_tag, True)

new_variant.set_editor_property(
                \"VariantChannelTag\",
                tagChannel)

inlineVariantDef = new_variant.get_editor_property(\"InlineVariant\")
richColorVar = inlineVariantDef.get_editor_property(\"RichColorVar\")

richColorVar.set_editor_property(
                \"ColorSwatchForChoices\",
                color_swatch_choice)

richColorVar.set_editor_property(
                \"ColorParamName\",
                asset_type_name + \"Color\")

variants.append(new_variant)" FortID _variantchannelname (Utilities|String|Append "Cosmetics.Variant.Channel.Vehicle." (Utilities|String|EnumtoString (Variables|Default|GetProductTypeSelection)) ".Color") _returnvalue (Utilities|String|EnumtoString (Variables|Default|GetProductTypeSelection)))
    (Task|SetTaskNotificationText (Utilities|Text|FormatText "INVTEXT(\"Creating RichColor Variant Property\")"))))

(fn CreateVariantsAssets)

(fn GenerateVariantsForFortID
  (params (CurrentFortID :type Fort%20Item%20Definition%20Object%20Reference) (Index :type Integer))
  (bind _index (Variables|GetIndex))
  (bind _paintnames (Variables|Default|GetPaintNames))
  (bind _output (Utilities|Array|Get(acopy) _paintnames _index))
  (bind _returnvalue (Utilities|String|ToString(Name) _output))
  (bind _returnvalue_2 (Utilities|String|Replace _returnvalue " "))
  (bind _output_get (LocalVariables|SetFortID CurrentFortID))
  (Utilities|Array|Clear (Variables|Default|GetCreatedFortVariants))
  (bind (_currentidname _currentvehiclename _currentpathandname) (CallFunction|GetRelevantStringsfromID _output_get))
  (bind _returnvalue_3 (EditorScripting|AssetTools|GetAssetTools))
  (bind _prefix (CallFunction|GetAssetPrefixName (Utilities|String|EnumtoString (Variables|Default|GetProductTypeSelection))))
  (bind _returnvalue_4 (EditorScripting|AssetTools|CreateAsset _returnvalue_3 (Utilities|String|Append _prefix "_" _currentvehiclename "_" _returnvalue_2) (Utilities|String|Append (Development|Editor|GetCurrentContentBrowserPath) "/" "Variants") (Utilities|GetClass "/Game/NeverCook/TechArt/VariantTool/_TemplateFiles/PaintedVariant_Template.PaintedVariant_Template") 0 "None" false))
  (bind _asfort_vehicle_cosmetics_variant_token_type (Utilities|Casting|CastToFortVehicleCosmeticsVariantTokenType _returnvalue_4))
  (Utilities|Array|Add (Variables|Default|GetCreatedFortVariants) _asfort_vehicle_cosmetics_variant_token_type)
  (Utilities|Array|AddUnique (Variables|Default|GetAllCreatedFortVariants) _asfort_vehicle_cosmetics_variant_token_type)
  (Python|Execution|ExecutePythonScript "import unreal

tagName = unreal.FortniteEditorLibrary.request_gameplay_tag(\"Cosmetics.Variant.Property.Vehicle.Painted.\" + my_color, True)
tagChannel = unreal.FortniteEditorLibrary.request_gameplay_tag(\"Cosmetics.Variant.Channel.Vehicle.Painted\", True)

my_asset.set_editor_property(
                \"grant_base_item\",
                True)

my_asset.set_editor_property(\"cosmetic_item\", fort_id)
my_asset.set_editor_property(\"VariantNameTag\", tagName)
my_asset.set_editor_property(\"VariantChannelTag\", tagChannel)
#my_asset.set_editor_property(\"Rarity\", rarity)" _asfort_vehicle_cosmetics_variant_token_type _returnvalue_2 _currentpathandname (LocalVariables|GetFortID) (Variables|Default|GetRaritySelection))
  (Task|SetTaskNotificationText (Utilities|Text|FormatText "INVTEXT(\"{step}/{max} - Creating Variant Colors\")" (+ (Variables|GetIndex) 1) (Utilities|Array|Length (Variables|Default|GetPaintNames)))))

(fn GetAssetPrefixName
  (params (AssetName :type String))
  (bind _localassetname (Variables|Default|GetLocalAssetName))
  (Variables|Default|SetLocalAssetName AssetName)
  (if (Utilities|String|Contains AssetName "Wheel")
    (return _localassetname)
    (else
      (return (Utilities|String|Append "Car" _localassetname)))))

(fn GetGameplayTagFromPaintName
  (params (PaintName :type Name))
  (bind _outtag (Class|FortniteEditorLibrary|RequestGameplayTag (Utilities|String|Replace (Utilities|String|Append "Cosmetics.Variant.Property.Vehicle.Painted." (Utilities|String|ToString(Name) PaintName)) " ") true))
  (return _outtag))

(fn GetGameplayTagFromProductType
  (bind _outtag (Class|FortniteEditorLibrary|RequestGameplayTag (Utilities|String|Append "VehicleCosmetics_" (Utilities|String|EnumtoString (Variables|Default|GetProductTypeSelection)))))
  (return _outtag))

(fn GetPaintedRowTag
  (bind _outtag (Class|FortniteEditorLibrary|RequestGameplayTag "CosmeticProperty.Vehicle.PaintedRow" true))
  (return _outtag))

(fn GetPreviewTextureForColor
  (params (PaintName :type Name))
  (bind _localpaintstring (LocalVariables|GetLocalPaintString))
  (bind _localpaintstring_2 (LocalVariables|GetLocalPaintString))
  (bind _localpaintstring_3 (LocalVariables|GetLocalPaintString))
  (LocalVariables|SetLocalPaintString (Utilities|String|ToString(Name) PaintName))
  (if (Utilities|String|Contains _localpaintstring "Burnt Sienna")
    (return "/VehicleCosmetics/UI/Icons/Painted/T_UI_Icon_Painted_BurntSiena.T_UI_Icon_Painted_BurntSiena")
    (elif (Utilities|String|Contains _localpaintstring "Black")
      (return "/VehicleCosmetics/UI/Icons/Painted/T_UI_Icon_Painted_Black.T_UI_Icon_Painted_Black")
      (elif (Utilities|String|Contains _localpaintstring "Cobalt")
        (return "/VehicleCosmetics/UI/Icons/Painted/T_UI_Icon_Painted_Cobalt.T_UI_Icon_Painted_Cobalt")
        (elif (Utilities|String|Contains _localpaintstring "Crimson")
          (return "/VehicleCosmetics/UI/Icons/Painted/T_UI_Icon_Painted_Crimson.T_UI_Icon_Painted_Crimson")
          (elif (Utilities|String|Contains _localpaintstring "Forest Green")
            (return "/VehicleCosmetics/UI/Icons/Painted/T_UI_Icon_Painted_ForestGreen.T_UI_Icon_Painted_ForestGreen")
            (elif (Utilities|String|Contains _localpaintstring_2 "Gold")
              (return "/VehicleCosmetics/UI/Icons/Painted/T_UI_Icon_Painted_Gold.T_UI_Icon_Painted_Gold")
              (elif (Utilities|String|Contains _localpaintstring_2 "Grey")
                (return "/VehicleCosmetics/UI/Icons/Painted/T_UI_Icon_Painted_Grey.T_UI_Icon_Painted_Grey")
                (elif (Utilities|String|Contains _localpaintstring_2 "Lime")
                  (return "/VehicleCosmetics/UI/Icons/Painted/T_UI_Icon_Painted_Lime.T_UI_Icon_Painted_Lime")
                  (elif (Utilities|String|Contains _localpaintstring_2 "Orange")
                    (return "/VehicleCosmetics/UI/Icons/Painted/T_UI_Icon_Painted_Orange.T_UI_Icon_Painted_Orange")
                    (elif (Utilities|String|Contains _localpaintstring_2 "Pink")
                      (return "/VehicleCosmetics/UI/Icons/Painted/T_UI_Icon_Painted_Pink.T_UI_Icon_Painted_Pink")
                      (elif (Utilities|String|Contains _localpaintstring_3 "Purple")
                        (return "/VehicleCosmetics/UI/Icons/Painted/T_UI_Icon_Painted_Purple.T_UI_Icon_Painted_Purple")
                        (elif (Utilities|String|Contains _localpaintstring_3 "Saffron")
                          (return "/VehicleCosmetics/UI/Icons/Painted/T_UI_Icon_Painted_Saffron.T_UI_Icon_Painted_Saffron")
                          (elif (Utilities|String|Contains _localpaintstring_3 "Sky Blue")
                            (return "/VehicleCosmetics/UI/Icons/Painted/T_UI_Icon_Painted_SkyBlue.T_UI_Icon_Painted_SkyBlue")
                            (elif (Utilities|String|Contains _localpaintstring_3 "Titanium White")
                              (return "/VehicleCosmetics/UI/Icons/Painted/T_UI_Icon_Painted_TitaniumWhite.T_UI_Icon_Painted_TitaniumWhite"))))))))))))))))

(fn GetRelevantStringsFromID
  (params (FortID :type Fort%20Item%20Definition%20Object%20Reference))
  (bind _returnvalue (Utilities|GetDisplayName FortID))
  (bind _currentselectionpath (Variables|Default|GetCurrentSelectionPath))
  (bind _returnvalue_2 (Utilities|String|Append _currentselectionpath "/" _returnvalue))
  (bind _returnvalue_3 (EditorScripting|Asset|LoadAsset _returnvalue_2 :!class /Script/EditorScriptingUtilities.EditorAssetLibrary))
  (Variables|Default|SetSelectedAssets (Utilities|Array|MakeArray _returnvalue_3))
  (return _returnvalue (Utilities|String|RightChop _returnvalue (+ (Utilities|String|FindSubstring _returnvalue "_") 1)) _returnvalue_2))

(fn GetVariantChannelName
  (params (InFortID :type Fort%20Item%20Definition%20Object%20Reference))
  (LocalVariables|SetFortID InFortID)
  (switch Utilities|FlowControl|Switch|SwitchonDelMar_ProductTypeEnum (Variables|Default|GetProductTypeSelection)
    (:NewEnumerator0
      _)
    (:NewEnumerator1
      _)
    (:NewEnumerator2
      _)
    (:NewEnumerator3
      (return "Decal Color"))
    (:NewEnumerator4
      _)
    (:NewEnumerator5
      _)
    (:NewEnumerator6
      _))
  (bind _vehiclename (Default|GetVehicleNamefromFortID (LocalVariables|GetFortID)))
  (return (Utilities|String|Append _vehiclename " " (Utilities|String|EnumtoString (Variables|Default|GetProductTypeSelection)) " Color")))

(fn GetVehicleNameFromFortID
  (params (FortID :type Fort%20Item%20Definition%20Object%20Reference))
  (return (Utilities|String|Replace (Utilities|String|RightChop (Utilities|String|ToString(Text) (ItemDefinition|GetDisplayName FortID)) (+ (Utilities|String|FindSubstring (Utilities|String|ToString(Text) (ItemDefinition|GetDisplayName FortID)) "_") 1)) ")")))

(fn MakeInstanceTableRow
  (params (RowName :type Name))
  (bind _returnvalue (Utilities|InstancedStruct|MakeInstancedStruct (Utilities|Struct|MakeCosmeticPropertyTableRow (Utilities|Struct|MakeDataTableRowHandle (Variables|Default|GetVehicleCosmeticDataTable) (Utilities|String|StringToName (Utilities|String|Replace (Utilities|String|ToString(Name) RowName) " "))) (GameplayTags|MakeLiteralGameplayTag "(TagName=\"CosmeticProperty.Vehicle.PaintedRow\")"))))
  (return _returnvalue))

(fn SetInitialVehicleCosmeticProperties
  (params (FortID :type Fort%20Item%20Definition%20Object%20Reference) (VehicleCosmetic :type Vehicle%20Cosmetics%20Item%20Definition%20Object%20Reference))
  (if (Utilities|Array|IsNotEmpty (Variables|Default|GetPaintNames))
    (bind _gameplaytag (Default|GetGameplayTagfromProductType))
    (bind _vehiclename (Default|GetVehicleNamefromFortID FortID))
    (Python|Execution|ExecutePythonScript "import unreal

variants = fort_id_asset.get_editor_property(\"ItemVariants\")
new_variant = unreal.FortCosmeticPropertyVariant(fort_id_asset)

tagChannel = unreal.FortniteEditorLibrary.request_gameplay_tag(\"Cosmetics.Variant.Channel.Vehicle.Painted\", True)

new_variant.set_editor_property(
                \"VariantChannelTag\",
                tagChannel)

loc_channel_name = unreal.NSLOCTEXT('', fort_id_asset.get_name() + \".\" + new_variant.get_name(), variant_channel_name)

new_variant.set_editor_property(
                \"VariantChannelName\",
                loc_channel_name)

variants.append(new_variant)
" FortID VehicleCosmetic _gameplaytag (Utilities|String|Append _vehiclename " Painted Style") (Utilities|Text|ToText(String) (Utilities|GetObjectName FortID)) (Utilities|String|EnumtoString (Variables|Default|GetProductTypeSelection))) ; Creating the array to put variants in\n
  ))

(fn SetVariantProperties
  (params (PaintName :type Name) (FortID :type Fort%20Item%20Definition%20Object%20Reference) (Variant :type Fort%20Vehicle%20Cosmetics%20Variant%20Token%20Type%20Object%20Reference) (VCIDAsset :type Vehicle%20Cosmetics%20Item%20Definition%20Object%20Reference))
  (bind _instancetablerow (Default|MakeInstanceTableRow PaintName))
  (bind _foundtexture (Default|GetPreviewTextureforColor PaintName))
  (bind _gameplaytag (Default|GetGameplayTagfromPaintName PaintName))
  (bind _gameplaytag_2 (Default|GetPaintedRowTag))
  (Python|Execution|ExecutePythonScript "import unreal

variants = fort_id_asset.get_editor_property(\"ItemVariants\")
new_variant = variants[0]

cosmetic_variant = new_variant.get_editor_property(\"GenericPropertyOptions\")

cosmetic_prop_def = unreal.CosmeticPropertyVariantDef()

cosmetic_prop_array = cosmetic_prop_def.get_editor_property(\"CosmeticProperties\")

cosmetic_prop_array.append(cosmetic_table_row)

if is_default_object == True:
    loc_variant_name = unreal.NSLOCTEXT('', fort_id_asset.get_name() + \".none\", paint_name)
else:
    loc_variant_name = unreal.NSLOCTEXT('', fort_id_asset.get_name() + \".\" + current_variant.get_name(), paint_name)

cosmetic_prop_def.set_editor_property(
                \"VariantName\",
                loc_variant_name)

cosmetic_prop_def.set_editor_property(
                \"CustomizationVariantTag\",
                cosmetic_tag)

cosmetic_prop_def.set_editor_property(
                \"start_unlocked\",
                False)

cosmetic_prop_def.set_editor_property(
                \"PreviewImage\",
                preview_texture)

if is_default_object == True:
    cosmetic_prop_def.set_editor_property(
                \"is_default\",
                True)

cosmetic_variant.append(cosmetic_prop_def)" Variant FortID (Utilities|String|ToString(Name) PaintName) _gameplaytag _gameplaytag_2 (Variables|Default|GetVehicleCosmeticDataTable) (== PaintName "") _instancetablerow _foundtexture (Utilities|String|EnumtoString (Variables|Default|GetProductTypeSelection))))
